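; Destructible dirt shell building prop that pays stone to players who damage it.
; Non-self damage grants the stone and fires a damage dispatcher; a custom leaf-server event takes a total health, lowering big health to it or destroying the prop at zero.
; ShouldDie returns a flag, broadcasting instead when death is blocked.
; Once playlist data is ready, a hotfix either disables damage and the health bar or the authority applies a max-health GameplayEffect.
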(event Custom|EdgeServerApplyDamage
  (params (TotalHealth :type Integer64))
  ;[1 1 1] Leaf server only
  (if (> TotalHealth 0)
    (CallFunction|SetBigHealth (Math|Integer64|Min(Integer64) (Variables|BuildingPropBigHealth|GetBigHealth) TotalHealth))
    (Networking|ForceNetUpdate)
    (else
      (CallFunction|DirtOutOfHealth))))

(event Custom|DirtOutOfHealth
  ;[1 1 1] Leaf server only
  ;;[1 1 1] DestroyActor on BuildingActor calls Die()
  (Actor|DestroyActor))

(event EventOnDamageServer
  (params (Damage :type Float%20%28single-precision%29) (DamageTags :type Gameplay%20Tag%20Container%20Structure%20%28by%20ref%29) (Momentum :type Vector :default "0, 0, 0") (HitInfo :type Hit%20Result%20Structure%20%28by%20ref%29) (InstigatedBy :type Controller%20Object%20Reference) (DamageCauser :type Actor%20Object%20Reference) (EffectContext :type Gameplay%20Effect%20Context%20Handle%20Structure))
  (|CallParentFunction Damage DamageTags Momentum HitInfo InstigatedBy DamageCauser EffectContext)
  (if (Utilities|Equal(Object) DamageCauser self)
    _
    (elif (> Damage 0.0)
      ;[1 1 1] This grants stone to the pawn doing damage to the dirt
      (Inventory|GiveItemtoInventoryOwner :InventoryOwner (Utilities|Casting|CastToFortPlayerController InstigatedBy) :ItemDefinition "/Game/Items/ResourcePickups/StoneItemData.StoneItemData" :NumberToGive 2 :bNotifyPlayer true :ItemLevel -1 :PickupInstigatorHandle -1 :bUseItemPickupAnalyticEvent false :WeaponAmmoOverride -1)
      (Default|CallOnDamageRecieved (Math|Float|Truncate Damage) self))))

(event EventBeginPlay
  (Collision|SetCollisionResponsetoChannel (Variables|Debug|GetStaticMeshComponent) "ECC_GameTraceChannel13" "ECR_Block")
  (Utilities|FlowControl|Sequence
    (:then_0
      (Class|FortAsyncActionCurrentPlaylistReady|PlaylistDataReadyAsync
        (:OnReady
          (if (== (ScalableFloat|GetValueatLevel (Variables|Default|GetHotfixNoDamage)) 1.0)
            (Fortnite|SetCanbeDamaged self)
            (Variables|BuildingPropBigHealth|SetbShowHealthBar false)
            (else
              (Networking|SwitchHasAuthority
                (:Authority
                  (GameplayEffects|ApplyGameplayEffectToSelf :self (Ability|GetAbilitySystemComponent self) :GameplayEffectClass "/Figment_S10/Athena/Prototype/Blueprints/Iceberg/GE_MaxHealthSet.GE_MaxHealthSet_C" :Level 0.0))))))))))

(event Collision|EventActorBeginOverlap
  (params (OtherActor :type Actor%20Object%20Reference)))

(event EventTick
  (params (DeltaSeconds :type Float%20%28single-precision%29)))

(fn ShouldDie
  (params (Damage :type Float%20%28single-precision%29) (EventInstigator :type Controller%20Object%20Reference) (DamageCauser :type Actor%20Object%20Reference) (bIsTest :type Boolean) (EffectContext :type Gameplay%20Effect%20Context%20Handle%20Structure%20%28by%20ref%29))
  (bind _propshoulddie (Variables|Default|GetPropShouldDie))
  (if (not _propshoulddie)
    (Default|CallOnShouldHaveDied self))
  (return _propshoulddie))

(fn ConstructionScript)
